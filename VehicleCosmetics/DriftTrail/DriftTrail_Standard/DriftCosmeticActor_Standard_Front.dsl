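; Drift trail cosmetic actor, standard front-wheel variant, that shows a color-customizable smoke particle effect while the vehicle drifts.
; When applied, it configures the smoke's auto-managed attachment to the parent actor's root at this actor's attach socket.
; Drift start and finish toggle the smoke, leaving the ground stops it, and landing mid-drift restarts it.
; Starts and restarts are gated by a front-wheel FX significance LOD check.
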(event Vehicle|Cosmetics|EventOnCosmeticApplied
  (bind _returnvalue (Actor|GetAttachParentActor))
  (bind _rootcomponent (Variables|/Script/Engine.Actor|GetRootComponent _returnvalue))
  (Utilities|IsValid _returnvalue
    (:Is%20Valid
      (Utilities|IsValid _rootcomponent
        (:Is%20Valid
          (Effects|Components|ParticleSystem|SetAutoAttachmentParameters (Variables|Default|GetSmoke) _rootcomponent (Actor|GetAttachParentSocketName) "SnapToTarget" "SnapToTarget" "SnapToTarget")
          (Effects|Components|ParticleSystem|SetUseAutoManageAttachment (Variables|Default|GetSmoke) true))
        (:Is%20Not%20Valid
          _)))
    (:Is%20Not%20Valid
      _))
  (Development|PrintString "Couldnt setup AutoManageAttachment" true true "(R=1.000000,G=0.848041,B=0.000000,A=1.000000)" 5.0))

(event EventBeginPlay
  (Variables|/Script/Engine.SceneComponent|SetbAbsoluteRotation false (Variables|Default|GetSmoke)))

(event Vehicle|Cosmetics|EventOnDriftStarted
  (bind _returnvalue (VehicleCosmetics|FX|GetMaxSignificanceLODFrontWheelDriftFX))
  (bind _btrigger (Class|BPVehicleCosmeticsLibrary|ShouldTriggerFX (Variables|Vehicle|Cosmetics|GetVehicleCosmeticsInterface) _returnvalue))
  (if _btrigger
    (Components|Activation|SetActive (Variables|Default|GetSmoke) true false)))

(event Vehicle|Cosmetics|EventOnDriftFinished
  (Components|Activation|SetActive (Variables|Default|GetSmoke) false false))

(event Vehicle|Cosmetics|EventOnVehicleWheelsLeftGround
  (if (Components|Activation|IsActive (Variables|Default|GetSmoke))
    (Components|Activation|Deactivate (Variables|Default|GetSmoke))))

(event Vehicle|Cosmetics|EventOnVehicleLanded
  (params (LandingForced :type Float%20%28single-precision%29) (bLandedKickflip :type Boolean))
  (if (Utilities|DoesObjectImplementInterface (Actor|GetOwner) "/Script/VehicleCosmeticsRuntime.VehicleCosmeticsInterface")
    (if (Vehicle|Cosmetics|Drifting|IsDrifting (Variables|Vehicle|Cosmetics|GetVehicleCosmeticsInterface))
      (bind _returnvalue (VehicleCosmetics|FX|GetMaxSignificanceLODFrontWheelDriftFX))
      (bind _btrigger (Class|BPVehicleCosmeticsLibrary|ShouldTriggerFX (Variables|Vehicle|Cosmetics|GetVehicleCosmeticsInterface) _returnvalue))
      (if _btrigger
        (Components|Activation|Activate (Variables|Default|GetSmoke) false)))))

(event Vehicle|Cosmetics|EventOnColorReady
  (params (Color :type Name :default "None") (AdditionalInfo :type Vehicle%20Cosmetics%20Additional%20Parameter%20Infos%20Structure))
  (Class|BPVehicleCosmeticsFunctionLibrary|CosmeticActorColorReady Color "color" AdditionalInfo (Variables|Default|GetSmoke)))

(fn ConstructionScript)
