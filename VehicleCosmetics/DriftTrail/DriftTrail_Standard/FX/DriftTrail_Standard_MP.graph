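# Niagara system for the standard drift trail vehicle cosmetic, rendering stylized tire smoke tinted by the painted color.
# Three active emitters layer panning tire-mesh smoke, wheel-well smoke pushed sideways by drift tilt, and stretched, speed-scaled trail sprites.
# Partitioned particles map to skeletal-mesh wheel sockets, letting FN_DisableTireFX mask individual tires.
# Frontend-preview switches adjust rates and motion; the speedometer and mesh-trail emitters are disabled.
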NiagaraSystem /VehicleCosmetics/DriftTrail/DriftTrail_Standard/FX/DriftTrail_Standard_MP
dependencies {
 usedDataInterfaces [
  &NiagaraDataInterfaceCurve
  &NiagaraDataInterfaceSpriteRendererInfo
  &NiagaraDataInterfaceSkeletalMesh
  &NiagaraDataInterfaceSocketReader
 ]
 usedDynamicInputs [&MakeFloat4FromVectorAndFloat]
 usedModules [
  &self:SystemSpawnScript.SystemScriptSource.SystemScriptGraph.NiagaraNodeAssignment_0.SetVariables_#1
  &self:SystemSpawnScript.SystemScriptSource.SystemScriptGraph.NiagaraNodeAssignment_3.SetVariables_#2
  &self:SystemSpawnScript.SystemScriptSource.SystemScriptGraph.NiagaraNodeAssignment_4.SetVariables_#3
  &self:SystemSpawnScript.SystemScriptSource.SystemScriptGraph.NiagaraNodeAssignment_1.SetVariables_#4
  &self:SystemSpawnScript.SystemScriptSource.SystemScriptGraph.NiagaraNodeAssignment_2.SetVariables_#5
  &FN_UnpackVehicleFlags
  /Game/Effects/Niagara/Modules/Vehicle/FN_UpdateVehicleAttribs
  &SystemState
 ]
}
emitters {
 PanningSmoke_TireMesh {
  emitterUpdateScript [
   &EmitterState()
   &SpawnRate(
    SpawnRate=&MakeCustomFloatFromBool(Bool=$User.bIsFrontendPreview, True Float=60, False Float=60)
    Use Spawn Probability=false
   )
   &SpawnBurst_Instantaneous(Spawn Count=2, Use Spawn Probability=false, UseLoopCountLimit=false) {enabled=false}
   &VC_WheelWrapScale(StandardWidth=50) {enabled=false, module=VC_WheelWrapScale001}
  ]
  particleSpawnScript [
   &VC_PaintedVariantsOverride() {enabled=false}
   &PartitionParticles()
   &InitializeParticle(Color=$User.PaintedColor, Lifetime=0.3, Sprite Size Mode=::NewEnumerator3, UsePositionOffset=false, Write Material Random=true) {module=InitializeParticle001}
  ]
  particleUpdateScript [
   &FN_DisableTireFX(
    Mask Index Attribute:NiagaraInt32=$Particles.PartitionParticles.Partition
    Mask Index FL=2
    Mask Index FR=3
    Mask Index RL=0
    Mask Index RR=1
   )
   &VC_WheelWrapScale(Wheel Index Attribute:NiagaraInt32=&Add_Integer(A=$Particles.PartitionParticles.Partition, B=2))
   &SkeletalMeshLocation(
    Bone / Socket Index=&Add_Integer(A=$Particles.PartitionParticles.Partition, B=2)
    Bone / Socket Sampling Mode=::NewEnumerator5
    MaskSpawnGroup=false
    Mesh Sampling Type=::NewEnumerator6
    Sampled Scale Coordinate Space:ENiagaraCoordinateSpace=::Simulation
    Skeletal Mesh=$System.SkeletalMesh_Wheels
   )
   &ScaleColor(
    Scale Alpha=&FloatFromCurve(
     FloatCurve=DataInterface() {propertyValues {Curve {keys [
         {interpMode=RCIM_Cubic, tangentMode=RCTM_User, time=0, value=0}
         {interpMode=RCIM_Cubic, tangentMode=RCTM_User, time=0.2, value=0.7}
         {tangentMode=RCTM_User, time=1, value=0}
        ]}}}
     Scale Curve=0.3
    )
    Scale RGB=&VectorFromFloat(Value=1)
    ScaleA=true
    ScaleRGB=true
   ) {module=ScaleColor001}
   &DynamicMaterialParameters(
    Index 2 Param 1=1
    Index 2 Param 2=&Add_Float(
     A=&FloatFromCurve(
      FloatCurve=DataInterface() {propertyValues {Curve {keys [{tangentMode=RCTM_User, time=0, value=0}, {tangentMode=RCTM_User, time=1, value=1}]}}}
      Scale Curve=0.6
     )
     B=&RandomRangeFloat(Minimum=0, Maximum=3)
    )
    Index 2 Param 3=&RemapRange(Input Value To Remap=$Particles.NormalizedAge, Input Min=0.5, Input Max=1, Output Min=1, Output Max=6)
    Index 2 Param 4=1
    Module.Index2Param0Write=true
    Module.Index2Param1Write=false
    Module.Index2Param2Write=false
    Module.Index2Param3Write=true
    Write Parameter Index 0=false
    Write Parameter Index 2=true
   ) {module=DynamicMaterialParameters001}
   &ParticleState()
   &SolveForcesAndVelocity(Clamp Velocity=false, Limit Acceleration=false) {enabled=false}
  ]
  rendererClasses [&NiagaraMeshRendererProperties]
  renderers [
   {rendererClass=&NiagaraMeshRendererProperties, rendererIndex=0}
   {rendererClass=&NiagaraMeshRendererProperties, rendererIndex=1}
   {rendererClass=&NiagaraMeshRendererProperties, rendererIndex=2}
  ]
 }
 Speedometer {
  bEnabled=false
  emitterSpawnScript [/SetVariables(Emitter.Digits=2, Emitter.DisplayScalar=true) {isSetParameters=true}]
  emitterUpdateScript [
   &EmitterState()
   &SpawnBurst_Instantaneous(Use Spawn Probability=false, UseLoopCountLimit=false)
  ]
  particleSpawnScript [&InitializeParticle(
    Color Mode=::NewEnumerator0
    Lifetime=$Emitter.CurrentLoopDuration
    Position Offset=<0 0 300>
    Sprite Size Mode=::NewEnumerator3
    Uniform Sprite Size=125
    UsePositionOffset=true
    Write Material Random=true
   )]
  particleUpdateScript [
   &ParticleState(Kill Particles When Lifetime Has Elapsed=false)
   &Color(Color=/Niagara/DynamicInputs/TypeConversions/MakeLinearColorFromFloat(Value:NiagaraFloat=&VectorLength(Vector=$Engine.Owner.Velocity))) {enabled=false}
   &Color(Color=&MakeLinearColorFromFloat(Value:NiagaraFloat=&SelectFloatFromArray(Direct Array Index=2, Float Selection Array=$User.Wheel_Width_Array))) {module=Color001}
  ]
  rendererClasses [&NiagaraSpriteRendererProperties]
  renderers [{rendererClass=&NiagaraSpriteRendererProperties}]
 }
 StylizedSmoke_WheelWell_ALL {
  emitterUpdateScript [
   &EmitterState(Life Cycle Mode=::NewEnumerator1, Loop Duration=1, UseLoopDelay=false)
   &SpawnRate(
    SpawnRate=&MakeCustomFloatFromBool(Bool=$User.bIsFrontendPreview, True Float=150, False Float=&Lerp_Float(A=250, B=250, Alpha=$System.Speed_Alpha))
    Use Spawn Probability=false
   )
   &SpawnBurst_Instantaneous(Spawn Count=2, Use Spawn Probability=false, UseLoopCountLimit=false) {enabled=false}
   &VC_WheelWrapScale() {enabled=false}
  ]
  particleSpawnScript [
   &VC_PaintedVariantsOverride() {enabled=false}
   &InitializeParticle(
    Color=$User.PaintedColor
    Lifetime=0.4
    Mesh Scale=&Multiply_VectorByFloat(Vector=$Emitter.WheelFXScale, Float=5)
    Mesh Scale Mode=::NewEnumerator4
    Sprite Rotation Mode=::NewEnumerator2
    Sprite Size Mode=::NewEnumerator3
    Uniform Sprite Size=&RandomRangeFloat(Minimum=10, Maximum=15)
    UsePositionOffset=false
    Write Material Random=true
   )
   &PartitionParticles(Number Of Partitions=4)
   /SetVariables(Particles.SpawnGroup=$Particles.PartitionParticles.Partition) {isSetParameters=true}
   &SkeletalMeshLocation(
    Bone / Socket Index=&Add_Integer(A=$Particles.PartitionParticles.Partition, B=0)
    Bone / Socket Sampling Mode=::NewEnumerator5
    Interpolate Mesh Sampling=false
    MaskSpawnGroup=false
    Mesh Sampling Type=::NewEnumerator6
    Orientation Sampling=::NewEnumerator1
    Skeletal Mesh=$System.SkeletalMesh_Wheels
   )
   &ShapeLocation(
    Cylinder Height=45
    Cylinder Radius=10
    MaskSpawnGroup=false
    Offset=&MakeVector(
     X=&Multiply_Float(
      A=&MakeFloatFromVector(Vector=$Particles.SkeletalMeshLocation.SampledScale, Channel=::NewEnumerator0)
      B=-30
     )
     Y=0
     Z=&Multiply_Float(
      A=&MakeFloatFromVector(Vector=$Particles.SkeletalMeshLocation.SampledScale, Channel=::NewEnumerator0)
      B=-35
     )
    )
    Offset Mode=::NewEnumerator0
    Override Seed=false
    Rotation Axis=<0 1 0>
    Rotation Mode=::NewEnumerator1
    Shape Primitive=::NewEnumerator1
    Surface Only=false
    UseOwnerScale=true
   ) {module=ShapeLocation001}
   &AddVelocity(Velocity=<1 0 0>, Velocity Speed Scale=125)
  ]
  particleUpdateScript [
   &FN_DisableTireFX(Mask Index Attribute:NiagaraInt32=$Particles.PartitionParticles.Partition)
   &ParticleState()
   &SpriteRotationRate(Rotation Rate=&RandomRangeFloat(Minimum=2, Maximum=5))
   &ScaleSpriteSize(
    Uniform Curve Scale=4
    Uniform Curve Sprite Scale=DataInterface() {propertyValues {Curve {keys [
        {arriveTangent=0.3427, interpMode=RCIM_Cubic, leaveTangent=0.3427, tangentMode=RCTM_User, time=0, value=0.15}
        {arriveTangent=1.2913, interpMode=RCIM_Cubic, leaveTangent=1.2913, tangentMode=RCTM_User, time=0.6, value=0.675}
        {interpMode=RCIM_Cubic, leaveTangent=0.6223, time=1, value=1.05}
       ]}}}
   )
   &ScaleColor(
    Scale Alpha=&FloatFromCurve(
     FloatCurve=DataInterface() {propertyValues {Curve {keys [
         {arriveTangent=-0.1101, interpMode=RCIM_Cubic, leaveTangent=-0.1101, tangentMode=RCTM_User, time=0, value=0}
         {arriveTangentWeight=0.0648, interpMode=RCIM_Cubic, leaveTangentWeight=0.3239, tangentMode=RCTM_User, time=0.1, value=1}
         {
          arriveTangent=-2.6546
          arriveTangentWeight=0.2491
          interpMode=RCIM_Cubic
          leaveTangent=-2.6546
          leaveTangentWeight=0.1993
          tangentMode=RCTM_User
          time=0.8
          value=0.6
         }
         {arriveTangent=0.0463, interpMode=RCIM_Cubic, leaveTangent=0.0463, tangentMode=RCTM_User, time=1, value=0}
        ]}}}
     Scale Curve=&Multiply_Float(A=$User.ScaleAlpha, B=1.5)
    )
    Scale RGB=&VectorFromFloat(Value=&FloatFromCurve(
      FloatCurve=DataInterface() {propertyValues {Curve {keys [
          {arriveTangent=-0.4623, interpMode=RCIM_Cubic, leaveTangent=-0.4623, tangentMode=RCTM_User, time=0, value=1}
          {interpMode=RCIM_Cubic, tangentMode=RCTM_User, time=0.5, value=0.8375}
         ]}}}
      Scale Curve=$User.ScaleColor
     ))
    ScaleA=true
    ScaleRGB=true
   )
   &DynamicMaterialParameters(
    Index 0 Param 3=1
    Index 0 Param 4=0.2
    Index 2 Param 1=1.5
    Index 2 Param 2=&Add_Float(
     A=&FloatFromCurve(FloatCurve=DataInterface() {propertyValues {Curve {keys [{tangentMode=RCTM_User, time=0, value=0}, {tangentMode=RCTM_User, time=1, value=1}]}}})
     B=&RandomRangeFloat(Minimum=0, Maximum=2)
    )
    Index 2 Param 3=&RemapRange(Input Value To Remap=$Particles.NormalizedAge, Input Min=0.5, Input Max=1, Output Min=0.75, Output Max=2)
    Index 2 Param 4=0.7
    Module.Index0Param0Write=true
    Module.Index0Param1Write=false
    Module.Index0Param2Write=true
    Module.Index0Param3Write=true
    Module.Index1Param0Write=true
    Module.Index1Param1Write=true
    Module.Index1Param2Write=true
    Module.Index1Param3Write=true
    Module.Index2Param0Write=true
    Module.Index2Param1Write=true
    Module.Index2Param2Write=true
    Module.Index2Param3Write=true
    Write Parameter Index 1=true
    Write Parameter Index 2=true
   )
   &LinearForce(Force=&Multiply_VectorByFloat(
     Vector=<-1 0 0.35>
     Float=&FloatFromCurve(
      FloatCurve=DataInterface() {propertyValues {Curve {keys [
          {arriveTangent=4.1785, interpMode=RCIM_Cubic, leaveTangent=4.1785, tangentMode=RCTM_User, time=0, value=0}
          {arriveTangent=0.0895, interpMode=RCIM_Cubic, leaveTangent=0.0895, tangentMode=RCTM_User, time=0.1, value=1}
          {arriveTangent=0.0224, interpMode=RCIM_Cubic, leaveTangent=0.0224, tangentMode=RCTM_User, time=1, value=1}
         ]}}}
      Scale Curve=2000
     )
    ))
   &LinearForce(Force=&Multiply_VectorByFloat(Vector=<0 1 0>, Float=&Multiply_Float(A=$System.VehicleDriftTilt, B=-1000))) {module=LinearForce003}
   &Drag(Drag=3, Use Linear Drag=true, Use Rotational Drag=false)
   &GravityForce(Gravity=<0 0 125>)
   &SolveForcesAndVelocity(Clamp Velocity=false, Limit Acceleration=false)
   /SetVariables() {isSetParameters=true}
   /SetVariables(
    Particles.ContactPosition:NiagaraPosition=&AddVectorToPosition(
     Position=$Particles.SkeletalMeshLocation.SampledPosition
     Vector=&MakeVector(
      X=&Multiply_Float(
       A=&MakeFloatFromVector(Vector=$Particles.SkeletalMeshLocation.SampledScale, Channel=::NewEnumerator0)
       B=-30
      )
      Y=0
      Z=&Multiply_Float(
       A=&MakeFloatFromVector(Vector=$Particles.SkeletalMeshLocation.SampledScale, Channel=::NewEnumerator0)
       B=-35
      )
     )
    )
    Particles.Scale=&Multiply_VectorByFloat(Vector=$Particles.SkeletalMeshLocation.SampledScale, Float=1)
   ) {isSetParameters=true}
   /SetVariables(
    Particles.SpikyPosition:NiagaraPosition=&AddVectorToPosition(Position=$Particles.ShapeLocation001.ShapePosition, Vector=<0 0 -10>)
    Particles.SpikySpriteSize:Vector2f=&Multiply_Vector2D_ByFloat(
     Float=&FloatFromCurve(
      FloatCurve=DataInterface() {propertyValues {Curve {keys [
          {arriveTangent=0.0853, interpMode=RCIM_Cubic, leaveTangent=0.0853, tangentMode=RCTM_User, time=0.1, value=0}
          {arriveTangent=0.9204, interpMode=RCIM_Cubic, leaveTangent=0.9204, tangentMode=RCTM_User, time=0.6, value=0}
          {arriveTangent=3.2555, interpMode=RCIM_Cubic, leaveTangent=3.2555, tangentMode=RCTM_User, time=0.9, value=0.95}
         ]}}}
      Scale Curve=&RandomRangeFloat(Minimum=0, Maximum=4)
     )
     Vector2D=<10 -44>
    )
    Particles.SpikySpriteVis:NiagaraInt32=&ModuloInt(A=&ReturnParticleIDAttributes(), B=3)
   ) {isSetParameters=true}
   /SetVariables(Particles.CameraOffset=55, Particles.UseTOD=true) {isSetParameters=true}
   &SubUVAnimation(
    Evaluation Type=::NewEnumerator0
    "Sprite Renderer "=DataInterface() {propertyValues {SpriteRenderer=&self:StylizedSmoke_WheelWell_ALL_1.NiagaraSpriteRendererProperties_3}}
    SubUV Animation Mode:ENiagara_SubUVLookupModeV2=::NewEnumerator2
    UseEndFrame=false
    UseStartFrame=false
   )
  ]
  rendererClasses [&NiagaraSpriteRendererProperties, &NiagaraMeshRendererProperties]
  renderers [
   {rendererClass=&NiagaraSpriteRendererProperties}
   {rendererClass=&NiagaraSpriteRendererProperties, rendererIndex=1}
   {rendererClass=&NiagaraSpriteRendererProperties, rendererIndex=2}
   {rendererClass=&NiagaraSpriteRendererProperties, rendererIndex=3}
   {rendererClass=&NiagaraMeshRendererProperties, rendererIndex=4}
  ]
 }
 stylized_smoke_mesh_trail {
  bEnabled=false
  emitterUpdateScript [
   &EmitterState()
   &SpawnPerUnit(
    Max Movement Threshold=20000
    Module.Use Max Movement Threshold=false
    Module.Use Movement Tolerance=false
    Spawn Spacing=120
    Use Spawn Probability=false
   )
   &SpawnRate(
    SpawnRate=&MakeCustomFloatFromBool(Bool=$User.bIsFrontendPreview, True Float=200, False Float=150)
    Use Spawn Probability=false
   )
  ]
  particleSpawnScript [
   &VC_PaintedVariantsOverride() {enabled=false}
   &InitializeParticle(
    Color=$User.PaintedColor
    Lifetime Max=0.5
    Lifetime Min=0.1
    Lifetime Mode=::NewEnumerator1
    Mesh Scale Mode=::NewEnumerator1
    Mesh Uniform Scale Max=4
    Mesh Uniform Scale Min=2
    Sprite Rotation Mode=::NewEnumerator2
    Sprite Size Mode=::NewEnumerator3
    UsePositionOffset=false
    Write Material Random=true
   )
   &PartitionParticles()
   &SkeletalMeshLocation(
    Bone / Socket Index=&Add_Integer(A=$Particles.PartitionParticles.Partition, B=4)
    Bone / Socket Sampling Mode=::NewEnumerator5
    Interpolate Mesh Sampling=false
    MaskSpawnGroup=false
    Mesh Sampling Type=::NewEnumerator6
    Skeletal Mesh=$System.SkeletalMesh
   )
   &InitialMeshOrientation(Use Rotation Vector=false)
   &AddVelocity(Velocity=&MakeCustomVectorFromBool(Bool=$User.bIsFrontendPreview, True Vector=<-1500 0 1>, False Vector=<1 0 0>))
  ]
  particleUpdateScript [
   &ParticleState()
   &ScaleMeshSize(Scale Factor=&VectorFromFloat(Value=&FloatFromCurve(
      FloatCurve=DataInterface() {propertyValues {Curve {keys [
          {arriveTangent=4.1785, interpMode=RCIM_Cubic, leaveTangent=4.1785, tangentMode=RCTM_User, time=0, value=0.1}
          {arriveTangent=1.1179, interpMode=RCIM_Cubic, leaveTangent=1.1179, tangentMode=RCTM_User, time=0.2, value=0.4}
          {arriveTangent=0.0224, interpMode=RCIM_Cubic, leaveTangent=0.0224, tangentMode=RCTM_User, time=1, value=1}
         ]}}}
      Scale Curve=0.6
     )))
   &UpdateMeshOrientation(Rotation Vector=&RandomRangeVector(Minimum=<-0.5 -0.5 -0.5>, Maximum=<0.5 0.5 0.5>))
   &ScaleColor(
    Scale Alpha=&FloatFromCurve(
     FloatCurve=DataInterface() {propertyValues {Curve {keys [
         {arriveTangent=-0.2506, interpMode=RCIM_Cubic, leaveTangent=-0.2506, tangentMode=RCTM_User, time=0.5, value=1.0113}
         {arriveTangent=-5.1259, interpMode=RCIM_Cubic, leaveTangent=-5.1259, tangentMode=RCTM_User, time=1, value=0}
        ]}}}
     Scale Curve=$User.ScaleAlpha
    )
    Scale RGB=&VectorFromFloat(Value=&FloatFromCurve(
      FloatCurve=DataInterface() {propertyValues {Curve {keys [
          {arriveTangent=-0.4623, interpMode=RCIM_Cubic, leaveTangent=-0.4623, tangentMode=RCTM_User, time=0, value=1}
          {interpMode=RCIM_Cubic, tangentMode=RCTM_User, time=0.5, value=0.25}
         ]}}}
      Scale Curve=$User.ScaleColor
     ))
    ScaleA=true
    ScaleRGB=true
   )
   &InheritSourceMovement(
    Applied Position Scale=&VectorFromFloat(Value=&FloatFromCurve(FloatCurve=DataInterface() {propertyValues {Curve {keys [<at 0 1>,<at 1 0>]}}}, Scale Curve=0.8))
    ApplyForce=false
    ApplyPosition=true
    ApplyPositionFalloff=false
   )
   &DynamicMaterialParameters(
    Index 0 Param 1=&FloatFromCurve(
     FloatCurve=DataInterface() {propertyValues {Curve {keys [
         {tangentMode=RCTM_User, time=0, value=1}
         {arriveTangent=-0.0405, interpMode=RCIM_Cubic, leaveTangent=-0.0405, tangentMode=RCTM_User, time=0.7009, value=1}
        ]}}}
     Scale Curve=&Multiply_Float(A=-100, B=$User.Global Scale)
    )
    Index 0 Param 2=&FloatFromCurve(
     FloatCurve=DataInterface() {propertyValues {Curve {keys [
         {tangentMode=RCTM_User, time=0, value=1}
         {arriveTangent=-0.0405, interpMode=RCIM_Cubic, leaveTangent=-0.0405, tangentMode=RCTM_User, time=0.4, value=1}
         {arriveTangent=-1.6592, interpMode=RCIM_Cubic, leaveTangent=-1.6592, tangentMode=RCTM_User, time=1, value=0.1}
        ]}}}
     Scale Curve=0.5
    )
    Index 0 Param 3=&RandomRangeFloat(
     Minimum=&FloatFromCurve(
      FloatCurve=DataInterface() {propertyValues {Curve {keys [{tangentMode=RCTM_User, time=0, value=0}, {tangentMode=RCTM_User, time=1, value=1}]}}}
      Scale Curve=&Multiply_Float(A=-500, B=$User.Global Scale)
     )
     Maximum=&FloatFromCurve(
      FloatCurve=DataInterface() {propertyValues {Curve {keys [
          {arriveTangent=2.9357, interpMode=RCIM_Cubic, leaveTangent=2.9357, tangentMode=RCTM_User, time=0, value=0.01}
          <at 0.3 0.55>
          {arriveTangent=1.7225, interpMode=RCIM_Cubic, leaveTangent=1.7225, tangentMode=RCTM_User, time=1, value=1}
         ]}}}
      Scale Curve=&Multiply_Float(A=-350, B=$User.Global Scale)
     )
    )
    Index 0 Param 4=&FloatFromCurve(FloatCurve=DataInterface() {propertyValues {Curve {keys [
         {arriveTangent=0.1175, interpMode=RCIM_Cubic, leaveTangent=0.1175, tangentMode=RCTM_Break, time=0.3, value=-1}
         {arriveTangent=2.3807, interpMode=RCIM_Cubic, leaveTangent=4.8504, tangentMode=RCTM_Break, time=0.7806, value=-0.4}
         {arriveTangent=8.9968, interpMode=RCIM_Cubic, leaveTangent=14.4841, tangentMode=RCTM_Break, time=1, value=0.8456}
        ]}}})
    Index 1 Param 1=&FloatFromCurve(
     FloatCurve=DataInterface() {propertyValues {Curve {keys [
         {arriveTangent=0.5348, interpMode=RCIM_Cubic, leaveTangent=0.5348, tangentMode=RCTM_User, time=0, value=0.02}
         {arriveTangent=261.9807, interpMode=RCIM_Cubic, leaveTangent=261.9807, tangentMode=RCTM_User, time=0.07, value=10}
         {arriveTangent=4038.106, interpMode=RCIM_Cubic, leaveTangent=4038.106, tangentMode=RCTM_User, time=0.125, value=150}
        ]}}}
     Scale Curve=&RandomRangeFloat(Minimum=0.6, Maximum=1.2)
    )
    Index 1 Param 2=&Divide_Float(A=&RandomRangeFloat(Minimum=0.015, Maximum=0.02), B=$User.Global Scale)
    Index 1 Param 3=&FloatFromCurve(
     FloatCurve=DataInterface() {propertyValues {Curve {keys [{interpMode=RCIM_Cubic, tangentMode=RCTM_User, time=0.4, value=0.5}, <at 0.8 0.4>]}}}
     Scale Curve=1.5
    )
    Index 1 Param 4=&Divide_Float(A=0.9, B=$User.Global Scale)
    Index 2 Param 1=&FloatFromCurve(
     FloatCurve=DataInterface() {propertyValues {Curve {keys [{tangentMode=RCTM_User, time=0, value=1}, {tangentMode=RCTM_User, time=1, value=1}]}}}
     Scale Curve=$User.Sparkles
    )
    Index 2 Param 2=&FloatFromCurve(
     FloatCurve=DataInterface() {propertyValues {Curve {keys [
         {interpMode=RCIM_Cubic, tangentMode=RCTM_User, time=0.3, value=0}
         {arriveTangent=2.4333, interpMode=RCIM_Cubic, leaveTangent=2.4333, tangentMode=RCTM_User, time=1, value=1}
        ]}}}
     Scale Curve=0.01
    )
    Index 2 Param 3=&FloatFromCurve(
     FloatCurve=DataInterface() {propertyValues {Curve {keys [{interpMode=RCIM_Cubic, tangentMode=RCTM_User, time=0.5, value=1}, {interpMode=RCIM_Cubic, tangentMode=RCTM_User, time=1, value=1}]}}}
     Scale Curve=0.0007
    )
    Module.Param0WriteEnabled=true
    Module.Param0WriteEnabled1=false
    Module.Param0WriteEnabled2=true
    Module.Param1WriteEnabled=true
    Module.Param1WriteEnabled1=true
    Module.Param1WriteEnabled2=true
    Module.Param2WriteEnabled=true
    Module.Param2WriteEnabled1=true
    Module.Param2WriteEnabled2=true
    Module.Param3WriteEnabled=true
    Module.Param3WriteEnabled1=true
    Module.Param3WriteEnabled2=false
    Write Parameter Index 1=true
    Write Parameter Index 2=true
   )
   &SolveForcesAndVelocity(Clamp Velocity=false, Limit Acceleration=false)
   &Color(Color=<rgba 0.2299 1 0.2304 1>)
  ]
  rendererClasses [&NiagaraMeshRendererProperties, &NiagaraSpriteRendererProperties]
  renderers [
   {rendererClass=&NiagaraMeshRendererProperties, rendererIndex=0}
   {rendererClass=&NiagaraSpriteRendererProperties, rendererIndex=1}
  ]
 }
 stylized_smoke_rear003 {
  emitterUpdateScript [
   &EmitterState()
   &SpawnPerUnit(
    Max Movement Threshold=20000
    Module.Use Max Movement Threshold=false
    Module.Use Movement Tolerance=false
    Spawn Spacing:NiagaraFloat=&Lerp_Float(A=15, B=5, Alpha=$System.Speed_Alpha)
    Use Spawn Probability=false
   )
   &SpawnRate(
    SpawnRate=/VehicleCosmetics/FX/Niagara/Modules/VC_Mask_Floats(
     FrontEnd Value:NiagaraFloat=300
     blsFrontEndPreview:NiagaraBool=$User.bIsFrontendPreview
     SpeedMin Value:NiagaraFloat=120
     SpeedMax Value:NiagaraFloat=180
     Speed Alpha:NiagaraFloat=$System.Speed_Alpha
    )
    Use Spawn Probability=false
   )
  ]
  particleSpawnScript [
   &VC_PaintedVariantsOverride() {enabled=false}
   &InitializeParticle(
    Color=&MultiplyLinearColor(
     Linear Color=$User.PaintedColor
     Scale Factor=&MakeVector4_FromLinearColor(LinearColor=/VehicleCosmetics/FX/Niagara/Modules/VC_Mask_LinearColor(Coro Color:LinearColor=<rgba 1 1 1 1>, Coro FrontEnd Color:LinearColor=<rgba 0.75 0.75 0.75 0.75>, FN Color:LinearColor=<rgba 1 1 1 1>))
    )
    Lifetime=&VC_Mask_Floats(Coro Value:NiagaraFloat=0.2, FrontEnd Value:NiagaraFloat=0.6, FN Value:NiagaraFloat=0.5)
    Mesh Scale Mode=::NewEnumerator1
    Mesh Uniform Scale Max=5
    Sprite Rotation Mode=::NewEnumerator2
    Sprite Size Mode=::NewEnumerator3
    Uniform Sprite Size=&Multiply_Float(
     A=&Multiply_Float(A=&RandomRangeFloat(Minimum=125, Maximum=170), B=0.3)
     B=&VC_Mask_Floats(Coro Value:NiagaraFloat=0.6, FrontEnd Value:NiagaraFloat=1, FN Value:NiagaraFloat=1)
    )
    UsePositionOffset=false
    Write Material Random=true
   )
   &PartitionParticles(Number Of Partitions=4)
   /SetVariables(Particles.SpawnGroup=$Particles.PartitionParticles.Partition) {isSetParameters=true}
   &SkeletalMeshLocation(
    Bone / Socket Index=&Add_Integer(A=$Particles.PartitionParticles.Partition, B=2)
    Bone / Socket Sampling Mode=::NewEnumerator5
    Interpolate Mesh Sampling=false
    MaskSpawnGroup=false
    Mesh Sampling Type=::NewEnumerator6
    Orientation Sampling=::NewEnumerator1
    Sampled Position Offset=&Add_Vector(
     A=&MaskVectorBySpawnGroup(
      Masked Vector=$System.TireGroundOffset
      Passthrough Vector=&MaskVectorBySpawnGroup(Masked Vector=$System.TireGroundOffset, Passthrough Vector=<0 0 -25>, Spawn Group Mask=2)
      Spawn Group Mask=3
     )
     B=&VC_Mask_Vectors(Coro Value:Vector3f=<-60 0 0>, FrontEnd Value:Vector3f=<0 0 0>, FN Value:Vector3f=<0 0 0>)
    )
    Skeletal Mesh=$System.SkeletalMesh
   )
   &AddVelocity(
    Velocity=&VC_Mask_Vectors(
     FrontEnd Value:Vector3f=<-1500 0 0>
     blsFrontEndPreview:NiagaraBool=$User.bIsFrontendPreview
     SpeedMin Value:Vector3f=<-800 0 0>
     SpeedMax Value:Vector3f=<-1 0 0>
     Speed Alpha:NiagaraFloat=$System.Speed_Alpha
    )
    Velocity Speed Scale=&MakeCustomFloatFromBool(
     Bool=&SetBoolByIntComparison(A=$Particles.PartitionParticles.Partition, B=1, Comparison Type=::NewEnumerator0)
     True Float=1
     False Float=0.1
    )
   )
   /SetVariables() {isSetParameters=true}
   /SetVariables(Particles.Initial.SpriteSize:Vector2f=$Particles.SpriteSize) {isSetParameters=true}
  ]
  particleUpdateScript [
   &FN_DisableTireFX(Mask Index Attribute:NiagaraInt32=$Particles.PartitionParticles.Partition)
   &ParticleState()
   &SpriteRotationRate(Rotation Rate=&RandomRangeFloat(Minimum=1, Maximum=2))
   &ScaleSpriteSize(
    Scale Sprite Size Mode=::NewEnumerator0
    Uniform Scale Factor=&MakeCustomFloatFromBool(
     Bool=$User.bIsFrontendPreview
     True Float=&FloatFromCurve(
      FloatCurve=DataInterface() {propertyValues {Curve {keys [
          {arriveTangent=3.3152, interpMode=RCIM_Cubic, leaveTangent=3.3152, tangentMode=RCTM_User, time=0.1, value=0.5}
          {arriveTangent=1.8429, interpMode=RCIM_Cubic, leaveTangent=1.8429, tangentMode=RCTM_User, time=0.1793, value=0.7202}
          {arriveTangent=0.3504, interpMode=RCIM_Cubic, leaveTangent=0.3504, tangentMode=RCTM_User, time=0.5121, value=1.0041}
          {arriveTangent=0.0224, interpMode=RCIM_Cubic, leaveTangent=0.0224, tangentMode=RCTM_User, time=1, value=1.1}
         ]}}}
      Scale Curve=4
     )
     False Float=&FloatFromCurve(
      FloatCurve=DataInterface() {propertyValues {Curve {keys [
          {arriveTangent=1.2879, interpMode=RCIM_Cubic, leaveTangent=1.2879, tangentMode=RCTM_User, time=0.05, value=0.15}
          {arriveTangent=2.1667, interpMode=RCIM_Cubic, leaveTangent=2.1667, time=0.2653, value=0.4863}
          {arriveTangent=2.5502, interpMode=RCIM_Cubic, leaveTangent=2.5502, tangentMode=RCTM_User, time=0.5, value=1.125}
          {arriveTangent=0.0224, interpMode=RCIM_Cubic, leaveTangent=0.0224, tangentMode=RCTM_User, time=1, value=1.5}
         ]}}}
      Scale Curve=4
     )
    )
   ) {module=ScaleSpriteSize002}
   &ScaleColor(
    Scale Alpha=&FloatFromCurve(
     FloatCurve=DataInterface() {propertyValues {Curve {keys [
         {arriveTangent=-2.5, interpMode=RCIM_Cubic, time=0, value=0}
         {arriveTangentWeight=0.0648, interpMode=RCIM_Cubic, leaveTangentWeight=0.3239, tangentMode=RCTM_User, time=0.2, value=1}
         {
          arriveTangent=-1.7281
          arriveTangentWeight=0.2491
          interpMode=RCIM_Cubic
          leaveTangent=-1.7281
          leaveTangentWeight=0.1993
          tangentMode=RCTM_User
          time=0.5
          value=0.7
         }
         {arriveTangent=-1.4112, interpMode=RCIM_Cubic, leaveTangent=-1.4112, tangentMode=RCTM_User, time=1, value=0}
        ]}}}
     Scale Curve=$User.ScaleAlpha
    )
    Scale RGB=&VectorFromFloat(Value=&FloatFromCurve(
      FloatCurve=DataInterface() {propertyValues {Curve {keys [
          {arriveTangent=-0.4623, interpMode=RCIM_Cubic, leaveTangent=-0.4623, tangentMode=RCTM_User, time=0, value=1}
          {arriveTangent=-0.7322, interpMode=RCIM_Cubic, leaveTangent=-0.7322, tangentMode=RCTM_User, time=0.5, value=0.6}
         ]}}}
      Scale Curve=$User.ScaleColor
     ))
    ScaleA=true
    ScaleRGB=true
   )
   &InheritSourceMovement(
    Applied Position Scale=&VectorFromFloat(Value=&MaskFloatBySpawnGroup(
      Masked Float=0.8
      Passthrough Float=&MaskFloatBySpawnGroup(
       Masked Float=0.8
       Passthrough Float=&FloatFromCurve(
        FloatCurve=DataInterface() {propertyValues {Curve {keys [<at 0 1 RCIM_Cubic>,<at 1 0>]}}}
        Scale Curve=&RandomRangeFloat(Minimum=0.75, Maximum=0.85)
       )
       Spawn Group Mask=1
      )
     ))
    ApplyForce=false
    ApplyPosition=true
    ApplyPositionFalloff=false
   ) {module=InheritSourceMovement001}
   &DynamicMaterialParameters(
    Index 0 Param 1=3
    Index 0 Param 3=1
    Index 0 Param 4=0.2
    Index 2 Param 1=2
    Index 2 Param 2=&Add_Float(
     A=&FloatFromCurve(
      FloatCurve=DataInterface() {propertyValues {Curve {keys [{tangentMode=RCTM_User, time=0, value=0}, {tangentMode=RCTM_User, time=1, value=1}]}}}
      Scale Curve=0.6
     )
     B=&RandomRangeFloat(Minimum=0, Maximum=3)
    )
    Index 2 Param 3=&RemapRange(Input Value To Remap=$Particles.NormalizedAge, Input Min=0.4, Input Max=1, Output Min=1, Output Max=3)
    Index 2 Param 4=&RandomRangeFloat(Minimum=0.6, Maximum=1)
    Index 3 Param 4=$Particles.SpriteRotation
    Module.Index0Param0Write=true
    Module.Index0Param1Write=false
    Module.Index0Param2Write=false
    Module.Index0Param3Write=true
    Module.Index2Param0Write=true
    Module.Index2Param1Write=true
    Module.Index2Param2Write=true
    Module.Index2Param3Write=true
    Module.Index3Param0Write=false
    Module.Index3Param1Write=false
    Module.Index3Param2Write=false
    Module.Index3Param3Write=true
    Write Parameter Index 2=true
    Write Parameter Index 3=true
   )
   &Drag(
    Drag=&MakeCustomFloatFromBool(Bool=true, True Float=1, False Float=3.5)
    Use Linear Drag=true
    Use Rotational Drag=false
   )
   &self:StretchAndFade(
    FloorMultiplier:NiagaraFloat=&RandomRangeFloat(Minimum=0.6, Maximum=0.8)
    ScaleMax=1255
    VelMax=200
    VelMin=50
   ) {module=StretchAndFade}
   &self:Face_Side_and_Up(Nth=3, UpVector:Vector3f=$Engine.Owner.SystemZAxis, Value:NiagaraInt32=$Particles.PartitionParticles.Partition) {module=Face_Side_and_Up}
   &GravityForce(Gravity=&MakeCustomVectorFromBool(
     Bool=&SetBoolByIntComparison(A=$Particles.UpVis, B=0, Comparison Type=::NewEnumerator0)
     True Vector=&Multiply_VectorByFloat(Vector=$Engine.Owner.SystemZAxis, Float=200)
     False Vector=<0 0 0>
    )) {module=GravityForce001}
   &SolveForcesAndVelocity(Clamp Velocity=false, Limit Acceleration=false)
   /SetVariables(Particles.ColorStretched:LinearColor=&MakeLinearColorFromVectorAndFloat(
     "Vector (RGB)"=&MakeVectorFromLinearColorRGB(Color=$Particles.Color)
     "Float (Alpha)"=&Multiply_Float(
      A=&FloatFromCurve(
       FloatCurve=DataInterface() {propertyValues {Curve {keys [
           {arriveTangent=0.7737, interpMode=RCIM_Cubic, leaveTangent=0.7737, tangentMode=RCTM_User, time=0, value=0}
           {arriveTangent=-0.2242, interpMode=RCIM_Cubic, leaveTangent=-0.2242, tangentMode=RCTM_User, time=0.6, value=0.9}
           {arriveTangent=-0.0405, interpMode=RCIM_Cubic, leaveTangent=-0.0405, tangentMode=RCTM_User, time=1, value=0}
          ]}}}
       Scale Curve=0.2
      )
      B=&MakeFloatFromLinearColor(LinearColor=$Particles.Color, Channel:ENiagaraLinearColor_Channels=::NewEnumerator3)
     )
    )) {isSetParameters=true}
   /SetVariables(Particles.GroundPosition:NiagaraPosition=&SubtractVectorFromPosition(Position=$Particles.Position, Vector=<0 0 10>)) {isSetParameters=true}
   /SetVariables(Particles.CamOff_Nage:NiagaraFloat=&FloatFromCurve(
     FloatCurve=DataInterface() {propertyValues {Curve {keys [{tangentMode=RCTM_User, time=0, value=0.2}, {tangentMode=RCTM_User, time=1, value=1}]}}}
     Scale Curve=25
    )) {isSetParameters=true}
   /SetVariables(Particles.OffsetPosition:NiagaraPosition=&AddVectorToPosition(
     Position=$Particles.Position
     Vector=&Multiply_VectorByFloat(
      Vector=$Engine.Owner.SystemZAxis
      Float=&Multiply_Float(A=&MakeFloatFromVector2d(Vector 2D=$Particles.SpriteSize, Channel=::NewEnumerator1), B=0.3)
     )
    )) {isSetParameters=true}
   /SetVariables(Particles.RearWheelVis:NiagaraInt32=&MakeCustomIntFromBool(
     Bool=&SetBoolByIntComparison(A=&SpawnGroup(), B=2, Comparison Type=::NewEnumerator2)
     True INT=&ModuloInt(A=&ReturnParticleIDAttributes(), B=3)
     False INT=0
    )) {isSetParameters=true}
   /SetVariables(Particles.Color_ThickAlpha:LinearColor=&MakeLinearColorFromVectorAndFloat(
     "Vector (RGB)"=&MakeVectorFromLinearColorRGB(Color=$Particles.Color)
     "Float (Alpha)"=&Multiply_Float(
      A=&MakeFloatFromLinearColor(LinearColor=<rgba 1 1 1 1>, Channel:ENiagaraLinearColor_Channels=::NewEnumerator3)
      B=&FloatFromCurve(
       FloatCurve=DataInterface() {propertyValues {Curve {keys [
           {tangentMode=RCTM_User, time=0, value=1}
           {arriveTangent=-0.0405, interpMode=RCIM_Cubic, leaveTangent=-0.0405, tangentMode=RCTM_User, time=0.85, value=1}
           {arriveTangent=-8.0089, interpMode=RCIM_Cubic, leaveTangent=-8.0089, tangentMode=RCTM_User, time=1, value=0}
          ]}}}
       Scale Curve=1.5
      )
     )
    )) {isSetParameters=true}
   /SetVariables(Particles.SparkleColor:LinearColor=<rgba 0.8281 0.6862 0.069 1>) {isSetParameters=true}
   /SetVariables(Particles.SparkleOffset=32) {isSetParameters=true}
   /SetVariables(Particles.SparklesVis:NiagaraInt32=&MakeCustomIntFromBool(
     Bool=&SetBoolByIntComparison(A=$User.ColorVariantID, B=1, Comparison Type=::NewEnumerator4)
     True INT=1
     False INT=0
    )) {isSetParameters=true}
   &Color(Color=&MaskLinearColorBySpawnGroup(
     Masked Linear Color=<rgba 1 0 0 1>
     Passthrough Linear Color=&MaskLinearColorBySpawnGroup(
      Masked Linear Color=<rgba 0 0.1312 1 1>
      Passthrough Linear Color=&MaskLinearColorBySpawnGroup(Masked Linear Color=<rgba 1 0 0.636 1>, Passthrough Linear Color=<rgba 0.9701 1 0 1>, Spawn Group Mask=2)
      Spawn Group Mask=1
     )
    )) {enabled=false, module=Color001}
  ]
  rendererClasses [&NiagaraSpriteRendererProperties]
  renderers [
   {rendererClass=&NiagaraSpriteRendererProperties}
   {rendererClass=&NiagaraSpriteRendererProperties, rendererIndex=1}
   {rendererClass=&NiagaraSpriteRendererProperties, rendererIndex=2}
   {rendererClass=&NiagaraSpriteRendererProperties, rendererIndex=3}
   {rendererClass=&NiagaraSpriteRendererProperties, rendererIndex=4}
   {rendererClass=&NiagaraSpriteRendererProperties, rendererIndex=5}
   {rendererClass=&NiagaraSpriteRendererProperties, rendererIndex=6}
   {rendererClass=&NiagaraSpriteRendererProperties, rendererIndex=7}
   {rendererClass=&NiagaraSpriteRendererProperties, rendererIndex=8}
   {rendererClass=&NiagaraSpriteRendererProperties, rendererIndex=9}
  ]
 }
}
summary {
 systemName=DriftTrail_Standard_MP
 emitters [
  {bEnabled=false, emitterName=Speedometer, rendererClasses [&NiagaraSpriteRendererProperties]}
  {emitterName=PanningSmoke_TireMesh, rendererClasses [&NiagaraMeshRendererProperties]}
  {emitterName=stylized_smoke_rear003, rendererClasses [&NiagaraSpriteRendererProperties]}
  {
   bEnabled=false
   emitterName=stylized_smoke_mesh_trail
   rendererClasses [&NiagaraMeshRendererProperties, &NiagaraSpriteRendererProperties]
  }
  {
   emitterName=StylizedSmoke_WheelWell_ALL
   rendererClasses [&NiagaraSpriteRendererProperties, &NiagaraMeshRendererProperties]
  }
 ]
 userVariables [
  User.bIsFrontendPreview {type=NiagaraBool, defaultValue=true}
  User.bIsPainted {type=NiagaraBool}
  User.bPreviewVehicleSpeed {type=NiagaraBool}
  User.ColorVariantID {type=NiagaraInt32}
  User.GameModeID {type=NiagaraInt32, defaultValue=3}
  User.Global Scale {type=NiagaraFloat, defaultValue=0, description=""}
  User.PaintedColor {type=LinearColor, description="", defaultValue {struct=LinearColor, value=<rgba 0.7188 0.7188 0.7188 1>}}
  User.PaintedScalar {type=NiagaraFloat, defaultValue=1, description=""}
  User.ScaleAlpha {type=NiagaraFloat, defaultValue=1, description=""}
  User.ScaleColor {type=NiagaraFloat, defaultValue=1, description=""}
  User.Sparkles {type=NiagaraFloat, defaultValue=0, description=""}
  User.VehicleFlags {type=NiagaraInt32, defaultValue=806289408}
  User.Wheel_Radius_Array {
   type=NiagaraDataInterfaceArrayFloat
   defaultValue {
    struct=NiagaraExt_VariableValue_DataInterface
    value {
     dataInterface=&self:NiagaraDataInterfaceArrayFloat_1
     dataInterfaceClass=&NiagaraDataInterfaceArrayFloat
    }
   }
  }
  User.Wheel_Width_Array {
   type=NiagaraDataInterfaceArrayFloat
   defaultValue {
    struct=NiagaraExt_VariableValue_DataInterface
    value {
     dataInterface=&self:NiagaraDataInterfaceArrayFloat_0
     dataInterfaceClass=&NiagaraDataInterfaceArrayFloat
    }
   }
  }
  User.WheelTouchingArray {
   type=NiagaraDataInterfaceArrayFloat
   defaultValue {
    struct=NiagaraExt_VariableValue_DataInterface
    value {
     dataInterface=&self:NiagaraDataInterfaceArrayFloat_2
     dataInterfaceClass=&NiagaraDataInterfaceArrayFloat
    }
   }
  }
 ]
}
