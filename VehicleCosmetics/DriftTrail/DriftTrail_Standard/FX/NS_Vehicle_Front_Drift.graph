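# Niagara system for the standard drift trail cosmetic's front-of-vehicle effect, drawing short-lived stylized smoke meshes.
# Its single emitter spawns randomly sized meshes by distance travelled, since rate spawning is disabled.
# Particles grow, rotate, darken, fade and follow source movement with diminishing influence.
# Curve-driven dynamic material parameters animate the shader over life, with user variables controlling color, alpha, scale and sparkles, though many declared ones are unreferenced here.
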NiagaraSystem /VehicleCosmetics/DriftTrail/DriftTrail_Standard/FX/NS_Vehicle_Front_Drift
dependencies {
 usedDataInterfaces [
  &NiagaraDataInterfaceMeshRendererInfo
  &NiagaraDataInterfaceCurve
  &NiagaraDataInterfaceVectorCurve
 ]
 usedDynamicInputs []
 usedModules [&SystemState]
}
emitters {stylized_smoke_mesh_trail {
  emitterUpdateScript [
   &EmitterState(Life Cycle Mode=::NewEnumerator1, Loop Duration=1, UseLoopDelay=false)
   &SpawnPerUnit(
    Max Movement Threshold=50000
    Module.Use Max Movement Threshold=false
    Module.Use Movement Tolerance=false
    Spawn Spacing=256
    Use Spawn Probability=false
   )
   &SpawnRate(SpawnRate=16, Use Spawn Probability=false) {enabled=false}
  ]
  particleSpawnScript [
   &InitializeParticle(
    Color=$User.color
    Lifetime Max=0.3
    Lifetime Min=0.2
    Lifetime Mode=::NewEnumerator1
    Mesh Renderer Array Visibility Mode=::NewEnumerator3
    Mesh Renderer Info=DataInterface() {propertyValues {MeshRenderer=&self:stylized_smoke_mesh_trail_3.NiagaraMeshRendererProperties_0}}
    Mesh Scale Mode=::NewEnumerator1
    Mesh Uniform Scale Max=5
    Mesh Uniform Scale Min=2.5
    UsePositionOffset=false
    Write Material Random=true
   )
   &InitialMeshOrientation(Use Rotation Vector=false)
   &AddVelocity(Velocity=&VectorFromFloat(Value=0.2), Velocity Speed Scale=0.2) {enabled=false}
   &AddVelocity(Velocity=<-500 0 0>) {enabled=false, module=AddVelocity001}
  ]
  particleUpdateScript [
   &ParticleState()
   &ScaleMeshSize(Scale Factor=&VectorFromFloat(Value=&FloatFromCurve(
      FloatCurve=DataInterface() {propertyValues {Curve {keys [
          {arriveTangent=2.0917, interpMode=RCIM_Cubic, leaveTangent=2.0917, tangentMode=RCTM_User, time=0, value=0.1}
          {arriveTangent=1.1961, interpMode=RCIM_Cubic, leaveTangent=1.1961, tangentMode=RCTM_User, time=0.2, value=0.45}
          {arriveTangent=0.0224, interpMode=RCIM_Cubic, leaveTangent=0.0224, tangentMode=RCTM_User, time=1, value=1}
         ]}}}
      Scale Curve=0.5
     )))
   &UpdateMeshOrientation(Rotation Vector=&RandomRangeVector(Minimum=<-0.5 -0.5 -0.5>, Maximum=<0.5 0.5 0.5>))
   &ScaleColor(
    Scale Alpha=&FloatFromCurve(
     FloatCurve=DataInterface() {propertyValues {Curve {keys [
         {arriveTangent=-0.3422, interpMode=RCIM_Cubic, leaveTangent=-0.3422, tangentMode=RCTM_User, time=0, value=1}
         {arriveTangent=-2.5714, interpMode=RCIM_Cubic, leaveTangent=-2.5714, tangentMode=RCTM_User, time=1, value=0}
        ]}}}
     Scale Curve=$User.ScaleAlpha
    )
    Scale RGB=&VectorFromFloat(Value=&FloatFromCurve(
      FloatCurve=DataInterface() {propertyValues {Curve {keys [
          {arriveTangent=-0.4623, interpMode=RCIM_Cubic, leaveTangent=-0.4623, tangentMode=RCTM_User, time=0, value=0.745}
          {interpMode=RCIM_Cubic, tangentMode=RCTM_User, time=0.5, value=0.15}
         ]}}}
      Scale Curve=$User.ScaleColor
     ))
    ScaleA=true
    ScaleRGB=true
   )
   &ScaleVelocity(Velocity Scale=&VectorFromCurve(
     VectorCurve=DataInterface() {propertyValues {XCurve {keys [<at 0 1>,<at 1 1>]}, YCurve {keys [<at 0 1>,<at 1 1>]}, ZCurve {keys [<at 0 1>,<at 1 250>]}}}
     Scale Curve=<1 1 2>
    ))
   &InheritSourceMovement(
    Applied Position Scale=&VectorFromFloat(Value=&FloatFromCurve(FloatCurve=DataInterface() {propertyValues {Curve {keys [<at 0 1>,<at 1 0>]}}}))
    ApplyForce=false
    ApplyPosition=true
    ApplyPositionFalloff=false
   )
   &InheritVelocity(Inherited Velocity Speed Limit=-100, Speed Limit=true, UseSpeedThreshold=false) {enabled=false}
   &SolveForcesAndVelocity(Clamp Velocity=false, Limit Acceleration=false)
   &DynamicMaterialParameters(
    Index 0 Param 1=&FloatFromCurve(
     FloatCurve=DataInterface() {propertyValues {Curve {keys [
         {tangentMode=RCTM_User, time=0, value=1}
         {arriveTangent=-0.0405, interpMode=RCIM_Cubic, leaveTangent=-0.0405, tangentMode=RCTM_User, time=0.7009, value=1}
        ]}}}
     Scale Curve=&Multiply_Float(A=-100, B=$User.Global Scale)
    )
    Index 0 Param 2=&FloatFromCurve(
     FloatCurve=DataInterface() {propertyValues {Curve {keys [
         {tangentMode=RCTM_User, time=0, value=1}
         {arriveTangent=-0.0405, interpMode=RCIM_Cubic, leaveTangent=-0.0405, tangentMode=RCTM_User, time=0.4, value=1}
         {arriveTangent=-1.6592, interpMode=RCIM_Cubic, leaveTangent=-1.6592, tangentMode=RCTM_User, time=1, value=0.1}
        ]}}}
     Scale Curve=0.5
    )
    Index 0 Param 3=&RandomRangeFloat(
     Minimum=&FloatFromCurve(
      FloatCurve=DataInterface() {propertyValues {Curve {keys [{tangentMode=RCTM_User, time=0, value=0}, {tangentMode=RCTM_User, time=1, value=1}]}}}
      Scale Curve=&Multiply_Float(A=-500, B=$User.Global Scale)
     )
     Maximum=&FloatFromCurve(
      FloatCurve=DataInterface() {propertyValues {Curve {keys [
          {arriveTangent=2.9357, interpMode=RCIM_Cubic, leaveTangent=2.9357, tangentMode=RCTM_User, time=0, value=0.01}
          <at 0.3 0.55>
          {arriveTangent=1.7225, interpMode=RCIM_Cubic, leaveTangent=1.7225, tangentMode=RCTM_User, time=1, value=1}
         ]}}}
      Scale Curve=&Multiply_Float(A=-350, B=$User.Global Scale)
     )
    )
    Index 0 Param 4=&FloatFromCurve(FloatCurve=DataInterface() {propertyValues {Curve {keys [
         {arriveTangent=0.1175, interpMode=RCIM_Cubic, leaveTangent=0.1175, tangentMode=RCTM_Break, time=0.3, value=-1}
         {arriveTangent=2.3807, interpMode=RCIM_Cubic, leaveTangent=4.8504, tangentMode=RCTM_Break, time=0.7806, value=-0.4}
         {arriveTangent=8.9968, interpMode=RCIM_Cubic, leaveTangent=14.4841, tangentMode=RCTM_Break, time=1, value=0.8456}
        ]}}})
    Index 1 Param 1=&FloatFromCurve(
     FloatCurve=DataInterface() {propertyValues {Curve {keys [
         {arriveTangent=0.5348, interpMode=RCIM_Cubic, leaveTangent=0.5348, tangentMode=RCTM_User, time=0, value=0.02}
         {arriveTangent=261.9807, interpMode=RCIM_Cubic, leaveTangent=261.9807, tangentMode=RCTM_User, time=0.07, value=10}
         {arriveTangent=4038.106, interpMode=RCIM_Cubic, leaveTangent=4038.106, tangentMode=RCTM_User, time=0.125, value=150}
        ]}}}
     Scale Curve=&RandomRangeFloat(Minimum=0.6, Maximum=1.2)
    )
    Index 1 Param 2=&Divide_Float(A=&RandomRangeFloat(Minimum=0.015, Maximum=0.02), B=$User.Global Scale)
    Index 1 Param 3=&FloatFromCurve(
     FloatCurve=DataInterface() {propertyValues {Curve {keys [{interpMode=RCIM_Cubic, tangentMode=RCTM_User, time=0.4, value=0.5}, <at 0.8 0.4>]}}}
     Scale Curve=1.5
    )
    Index 1 Param 4=&Divide_Float(A=0.9, B=$User.Global Scale)
    Index 2 Param 1=&FloatFromCurve(
     FloatCurve=DataInterface() {propertyValues {Curve {keys [
         {interpMode=RCIM_Cubic, tangentMode=RCTM_User, time=0.4, value=0}
         {arriveTangent=1.3341, interpMode=RCIM_Cubic, leaveTangent=1.3341, tangentMode=RCTM_User, time=1, value=1}
        ]}}}
     Scale Curve=$User.Sparkles
    )
    Index 2 Param 2=&FloatFromCurve(
     FloatCurve=DataInterface() {propertyValues {Curve {keys [
         {interpMode=RCIM_Cubic, tangentMode=RCTM_User, time=0.3, value=0}
         {arriveTangent=2.4333, interpMode=RCIM_Cubic, leaveTangent=2.4333, tangentMode=RCTM_User, time=1, value=1}
        ]}}}
     Scale Curve=0.01
    )
    Index 2 Param 3=&FloatFromCurve(
     FloatCurve=DataInterface() {propertyValues {Curve {keys [{interpMode=RCIM_Cubic, tangentMode=RCTM_User, time=0.5, value=1}, {interpMode=RCIM_Cubic, tangentMode=RCTM_User, time=1, value=1}]}}}
     Scale Curve=0.0007
    )
    Module.Param0WriteEnabled=true
    Module.Param0WriteEnabled1=false
    Module.Param0WriteEnabled2=true
    Module.Param1WriteEnabled=true
    Module.Param1WriteEnabled1=true
    Module.Param1WriteEnabled2=true
    Module.Param2WriteEnabled=true
    Module.Param2WriteEnabled1=true
    Module.Param2WriteEnabled2=true
    Module.Param3WriteEnabled=true
    Module.Param3WriteEnabled1=true
    Module.Param3WriteEnabled2=false
    Write Parameter Index 1=true
    Write Parameter Index 2=true
   )
  ]
  rendererClasses [&NiagaraMeshRendererProperties]
  renderers [{rendererClass=&NiagaraMeshRendererProperties, rendererIndex=0}]
 }}
summary {
 systemName=NS_Vehicle_Front_Drift
 emitters [{emitterName=stylized_smoke_mesh_trail, rendererClasses [&NiagaraMeshRendererProperties]}]
 userVariables [
  User.color {type=LinearColor, description="", defaultValue {struct=LinearColor, value=<rgba 0.7917 0.7917 0.7917 0.75>}}
  User.EmissiveColor {type=LinearColor, description="", defaultValue {struct=LinearColor, value=<rgba 1 1 1 1>}}
  User.EmissiveIntensity {type=NiagaraFloat, defaultValue=1, description=""}
  User.FireErosionDelay {type=NiagaraFloat, defaultValue=0, description=""}
  User.FireErosionDuration {type=NiagaraFloat, defaultValue=1, description=""}
  User.Global Scale {type=NiagaraFloat, defaultValue=1, description=""}
  User.ScaleAlpha {type=NiagaraFloat, defaultValue=0.8, description=""}
  User.ScaleColor {type=NiagaraFloat, defaultValue=1, description=""}
  User.smokeDif {type=NiagaraFloat, defaultValue=1, description=""}
  User.SmokeLifeTime {type=NiagaraFloat, defaultValue=1, description=""}
  User.SmokeParticleScale {type=NiagaraFloat, defaultValue=1, description=""}
  User.SmokeRandomLifeTime {type=NiagaraFloat, defaultValue=1, description=""}
  User.SmokeTextureScale {type=NiagaraFloat, defaultValue=0, description=""}
  User.Sparkles {type=NiagaraFloat, defaultValue=0, description=""}
  User.speed {type=NiagaraFloat, defaultValue=-1.5, description=""}
 ]
}
