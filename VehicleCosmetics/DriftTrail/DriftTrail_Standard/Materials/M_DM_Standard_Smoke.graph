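# Translucent unlit material for standard drift trail tire smoke, built for Niagara mesh particles.
# Stylized shading posterizes camera-vector and sky-light dot products, alongside time-of-day-scaled base color, particle color and dynamic-parameter-driven noise sparkles.
# Opacity uses normal-perturbed Fresnel and particle alpha, with either depth and camera-distance fades or a static-switched temporal dither.
# Panning noise displaces vertices along their normals, masked by vertex color.
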Material /VehicleCosmetics/DriftTrail/DriftTrail_Standard/Materials/M_DM_Standard_Smoke
graph {
 properties {
  bUsedWithNiagaraMeshParticles=true
  blendMode=BLEND_Translucent
  pixelDepthOffsetMode=PDOM_Legacy
  previewMesh=/VehicleCosmetics/DriftTrail/DriftTrail_Standard/Meshes/SM_TireSmoke_02
  shadingModel=MSM_Unlit
  translucencyPass=MTP_AfterMotionBlur
  usedShadingModels=Unlit
 }
 nodes [
  n1 = MaterialFunctionCall() {materialFunction=/Game/Packages/Fortress_SharedMaterials/Base_Material_Functions/SkylightColorAndIntensity.SkyLightColorAndIntensity}
  n2 = Multiply(A=@n1.Color, B=@n1.Color)
  Add(A=@n2) {constB=0.1}
  Constant() {r=1}
  n3 = NamedRerouteUsage()
  n4 = NamedRerouteUsage()
  n5 = Multiply(A=Add(A=NamedRerouteUsage(), B=Multiply(A=@n3, B=@n4) {constA=1}), B=ParticleColor().RGB)
  MaterialFunctionCall(Emissive=Multiply(A=@n5)) {materialFunction=&MF_FX_EmissiveAdjustments}
  Multiply(A=@n5) {constB=0.4}
  Multiply(A=@n2) {constB=0}
  n6 = Time()
  n7 = ScalarParameter() {defaultValue=0.5, group=WPO, parameterName=WPO Speed}
  n8 = TextureSample(UVs=Panner(Coordinate=TextureCoordinate() {uTiling=0.5, vTiling=0.5}, Time=Multiply(A=@n6, B=@n7)) {speedX=0.1, speedY=0.1}) {samplerType=SAMPLERTYPE_Normal, texture=/Game/Global/Textures/Default/Tileable_Patterns/2ndInfectedMapNormal}
  n9 = Multiply(A=Clamp(None=LinearInterpolate(Alpha=@n8.R)), B=ScalarParameter() {defaultValue=1, group=WPO, parameterName=WPO Intensity})
  n10 = VertexNormalWS()
  NamedRerouteDeclaration(None=Multiply(A=Multiply(A=@n9, B=@n10), B=VertexColor())) {name=WPO, nodeColor=<rgba 0 1 0.1328 1>}
  n11 = CameraVectorWS()
  n12 = ScalarParameter() {
   defaultValue=0.7
   desc=Adjustment to highlight one side of the contour based on the camera vector.
   group=Shading
   parameterName=Shading Adjustment
  }
  n13 = LinearInterpolate(A=Multiply(A=Transform(None=@n11) {transformSourceType=TRANSFORMSOURCE_World, transformType=TRANSFORM_Camera}), B=@n11, Alpha=@n12)
  n14 = PixelNormalWS()
  n15 = Multiply(A=SkyAtmosphereLightDirection(), B=ScalarParameter() {defaultValue=0.5, group=Shading, parameterName=SkyAtmosphereLightDir_BlendRate}) {constB=0.333}
  n16 = ComponentMask(None=@n15) {g=true, r=true}
  n17 = LinearInterpolate(A=ComponentMask(None=@n15) {b=true}) {constAlpha=0.8}
  n18 = LinearInterpolate(
   A=OneMinus(None=DotProduct(A=@n13, B=@n14))
   B=DotProduct(A=@n14, B=AppendVector(A=@n16, B=@n17))
   Alpha=ScalarParameter() {parameterName=Sky_Light_Posterize}
  )
  NamedRerouteDeclaration(None=MaterialFunctionCall!(Posterize=ScalarParameter() {defaultValue=2, group=Base, parameterName=Posterize}, Input=Clamp(None=@n18) {minDefault=0.05}) {materialFunction=/DealershipSharedLibrary/FX/Materials/MaterialFunctions/MF_Simple_Posterize}.Result) {name=FInal_Posterize, nodeColor=<rgba 1 0.0234 0 1>}
  n19 = ParticleColor()
  n20 = ScalarParameter() {defaultValue=1, parameterName=Opacity_U}
  n21 = ScalarParameter() {defaultValue=1, parameterName=Opacity_V}
  n22 = Panner(Coordinate=Multiply(A=TextureCoordinate(), B=AppendVector(A=@n20, B=@n21))) {bFractionalPart=true, speedY=0.02}
  n23 = Multiply(
   A=Normalize(VectorInput=TextureSample(UVs=@n22) {samplerType=SAMPLERTYPE_Normal, texture=&2ndInfectedMapNormal}.RGB)
   B=VectorParameter() {parameterName=NormalIntensity, defaultValue=<rgba 0.3 0.3 1.5 0>}.RGB
  )
  n24 = ScalarParameter() {defaultValue=1, parameterName=2ndFresnelLerp}
  n25 = MaterialFunctionCall(
   Brightness=ScalarParameter() {defaultValue=1.5, parameterName=FresnelBrightness}
   Exponent=ScalarParameter() {defaultValue=10, parameterName=FresnelExponent}
   Vertex / Pixel Normal=LinearInterpolate(
    A=Transform(None=@n23)
    B=VertexNormalWS()
    Alpha=MaterialFunctionCall(Exponent=@n24) {materialFunction=&MF_Fresnel_FX}.Fresnel
   )
  ) {materialFunction=&MF_Fresnel_FX}
  n26 = Multiply(A=@n19.A, B=@n25.Fresnel)
  n27 = ScalarParameter() {defaultValue=1, group=Opacity, parameterName=Dither Offset}
  n28 = ScalarParameter() {defaultValue=1, group=Opacity, parameterName=Dither Random}
  n29 = ScalarParameter() {defaultValue=1, group=Opacity, parameterName=Dither Offset 2}
  n30 = MaterialFunctionCall!(
   Distance Bias=ScalarParameter() {defaultValue=50, parameterName=FogBias}
   Distance Divisor=ScalarParameter() {defaultValue=150, parameterName=FogDivisor}
   Scene Depth
   Pixel Depth
  ) {materialFunction=/Game/Effects/Fort_Effects/Materials/_Functions/MF_DistanceBehindTranslucency_01}
  n31 = MaterialFunctionCall(
   Opacity In=Multiply(A=@n19.A, B=Multiply(A=@n25.Fresnel, B=@n30.Clamped))
   Position=WorldPosition().XYZ
   Fade_Start_Distance=ScalarParameter() {defaultValue=40, parameterName=CameraFadeDistanceStart}
   Fade_End_Distance=ScalarParameter() {defaultValue=15, parameterName=CameraFadeDistanceEnd}
  ) {materialFunction=&MF_Distance_SyncedFade_FX}
  n32 = ScalarParameter() {defaultValue=3, parameterName=Fade}
  NamedRerouteDeclaration(None=StaticSwitchParameter(
    True=MaterialFunctionCall(Alpha Threshold=@n26, Offset=@n27, Random=@n28, Offset2=@n29) {materialFunction=&DitherTemporalAA_Fortnite}.Result
    False=Multiply(A=@n31.Result, B=@n32)
   ) {parameterName=Use_DitherAA}) {name=Final_Opacity, nodeColor=<rgba 1 0 0.6562 1>}
  n33 = Multiply(
   A=TextureCoordinate()
   B=AppendVector(A=ScalarParameter() {defaultValue=1, parameterName=Base_U}, B=ScalarParameter() {defaultValue=1, parameterName=Base_V})
  )
  n34 = Add(A=TextureSample(UVs=Panner(Coordinate=@n33) {bFractionalPart=true, speedY=0.02}) {texture=&2ndInfectedMapLinear}.RGB)
  n35 = StaticSwitchParameter(
   True=LinearInterpolate(A=VectorParameter() {parameterName=BaseColorLow, defaultValue=<rgba 0.0417 0.0417 0.0417 1>}.RGB, Alpha=Divide(A=@n34) {constB=2.5}) {constA=0.5}
   False=Constant() {r=1}
  ) {defaultValue=true, parameterName=Use_Base_Color}
  n36 = ScalarParameter() {defaultValue=1, group=Shading, parameterName=Morning Scalar}
  n37 = ScalarParameter() {defaultValue=1, group=Shading, parameterName=Day Scalar}
  n38 = ScalarParameter() {defaultValue=1, group=Shading, parameterName=Evening Scalar}
  n39 = ScalarParameter() {defaultValue=1, group=Shading, parameterName=Night Scalar}
  NamedRerouteDeclaration(None=MaterialFunctionCall(
    Color_Morning=Multiply(A=@n35, B=@n36)
    Color_Day=Multiply(A=@n35, B=@n37)
    Color_Evening=Multiply(A=@n35, B=@n38)
    Color_Night=Multiply(A=@n35, B=@n39)
   ) {materialFunction=&Color_Time_of_Day}.Result) {name=Base_Color, nodeColor=<rgba 0 0.2969 1 1>}
  n40 = TextureCoordinate() {uTiling=3, vTiling=3}
  n41 = Panner() {speedY=0.005}
  n42 = Multiply(A=ComponentMask(None=TextureSample(UVs=@n41) {texture=/DealershipSharedLibrary/FX/Textures/Noise/T_VCVFX_Noise01_A}.RGB) {g=true, r=true}) {constB=0.0625}
  n43 = Multiply(A=Multiply(
    A=TextureSample(UVs=@n40) {texture=/DealershipSharedLibrary/FX/Textures/Noise/T_VCVFX_NoiseGeneric01Pack_RGB}.B
    B=TextureSample(UVs=@n42) {texture=&T_VCVFX_NoiseGeneric01Pack_RGB}.B
   )) {constB=4096}
  NamedRerouteDeclaration(None=Multiply(
    A=Multiply(A=@n43, B=@n19.RGB)
    B=DynamicParameter() {parameterIndex=2, defaultValue=<rgba 0 1 1 1>, paramNames [Sparkles, "", "", ""]}.Sparkles
   )) {name=Sparkles Result, nodeColor=<rgba 0 0.5781 1 1>}
  NamedRerouteUsage()
  NamedRerouteUsage()
 ]
}
