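# Unlit masked Niagara sprite material that draws shimmering sparkles for the standard drift trail vehicle cosmetic.
# A slowly panning noise texture's red and green channels become UVs for a packed-noise lookup, multiplied by a static sample of that channel.
# The heavily boosted product is tinted by particle color and feeds a shared emissive adjustment function.
# That same result, multiplied by a soft circular mask and particle alpha, forms a second output, apparently the opacity mask.
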Material /VehicleCosmetics/DriftTrail/DriftTrail_Standard/Materials/M_DM_Standard_Sparkles
graph {
 properties {
  bUsedWithNiagaraSprites=true
  blendMode=BLEND_Masked
  pixelDepthOffsetMode=PDOM_Legacy
  shadingModel=MSM_Unlit
  translucencyPass=MTP_AfterMotionBlur
  usedShadingModels=Unlit
 }
 nodes [
  n1 = TextureCoordinate()
  n2 = ComponentMask(None=TextureSample(UVs=Panner(Coordinate=@n1) {speedY=0.005}) {texture=/DealershipSharedLibrary/FX/Textures/Noise/T_VCVFX_Noise01_A}.RGB) {g=true, r=true}
  n3 = Multiply(
   A=TextureSample(UVs=TextureCoordinate()) {texture=/DealershipSharedLibrary/FX/Textures/Noise/T_VCVFX_NoiseGeneric01Pack_RGB}.B
   B=TextureSample(UVs=Multiply(A=@n2) {constB=0.0625}) {texture=&T_VCVFX_NoiseGeneric01Pack_RGB}.B
  )
  n4 = ParticleColor()
  n5 = Multiply(A=Multiply(A=Multiply(A=@n3) {constB=4096}, B=@n4.RGB)) {constB=0.75}
  MaterialFunctionCall(Emissive=@n5) {materialFunction=&MF_FX_EmissiveAdjustments}
  n6 = TextureCoordinate()
  n7 = Constant() {r=0.5}
  n8 = Constant() {r=0.35}
  Multiply(A=@n4.A, B=Multiply(A=@n5, B=SphereMask(A=@n6, B=@n7, Radius=@n8) {attenuationRadius=100, hardnessPercent=0}))
 ]
}
